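# Niagara system for water slide splash effects in the Hera theme park environment.
# It holds two Splash emitters and three generically named Empty emitters, one of which is disabled.
# A User.BoxSize vector defaults to a long, narrow, flat extent, though its binding to the emitters is not shown.
# The only listed module is SystemState, and no emitter lists a renderer, so particle behavior and visuals are not visible here.
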NiagaraSystem /Game/Environments/Hera/FX/ThemePark/NSLW_ThemePark_WaterSlide_Splash
dependencies {usedDataInterfaces [], usedDynamicInputs [], usedModules [&SystemState]}
emitters {
 Empty {rendererClasses [], renderers []}
 Empty001 {bEnabled=false, rendererClasses [], renderers []}
 Empty003 {rendererClasses [], renderers []}
 Splash {rendererClasses [], renderers []}
 Splash001 {rendererClasses [], renderers []}
}
summary {
 systemName=NSLW_ThemePark_WaterSlide_Splash
 emitters [
  {emitterName=Splash, rendererClasses []}
  {emitterName=Splash001, rendererClasses []}
  {emitterName=Empty, rendererClasses []}
  {bEnabled=false, emitterName=Empty001, rendererClasses []}
  {emitterName=Empty003, rendererClasses []}
 ]
 userVariables [User.BoxSize {type=Vector3f, description="", defaultValue {struct=Vector3f, value=<10 800 0>}}]
}
