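# Niagara system for a river water spout effect in the Hera environment's theme park.
# It holds four emitters: SmokeLeft and Splash, named for smoke and splash elements, plus Empty001 and Empty003.
# SystemState is the only listed module, and no renderers, data interfaces, dynamic inputs or user variables are recorded.
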NiagaraSystem /Game/Environments/Hera/FX/ThemePark/NS_ThemePark_WaterSpout_River
dependencies {usedDataInterfaces [], usedDynamicInputs [], usedModules [&SystemState]}
emitters {
 Empty001 {rendererClasses [], renderers []}
 Empty003 {rendererClasses [], renderers []}
 SmokeLeft {rendererClasses [], renderers []}
 Splash {rendererClasses [], renderers []}
}
summary {
 systemName=NS_ThemePark_WaterSpout_River
 emitters [
  {emitterName=SmokeLeft, rendererClasses []}
  {emitterName=Splash, rendererClasses []}
  {emitterName=Empty003, rendererClasses []}
  {emitterName=Empty001, rendererClasses []}
 ]
 userVariables []
}
